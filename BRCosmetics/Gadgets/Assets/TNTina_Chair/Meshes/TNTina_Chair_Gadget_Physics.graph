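# Physics asset for the skeletal mesh of the TNTina chair gadget, defining its simplified collision and joint limits.
# Three bodies, each a single capsule, cover the Anim, seat and seat_back bones.
# Two constraints chain seat_back to seat and seat to Anim, with uniform angular limit values on both swing axes and twist.
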PhysicsAsset /BRCosmetics/Gadgets/Assets/TNTina_Chair/Meshes/TNTina_Chair_Gadget_Physics
bodies {
 Anim [{shapeType=Capsule, length=29.7804, radius=29.6366, center=<-0.0272 0.173 21.2592>, rotation=<rot 0.1114 167.6447 -90>}]
 seat [{shapeType=Capsule, length=38.3035, radius=33.1983, center=<-1.6237 0.3065 20.3243>, rotation=<rot 0 -0.723 -89.8421>}]
 seat_back [{shapeType=Capsule, length=25.4906, radius=18.4331, center=<5.0014 -0.3228 2.913>, rotation=<rot 0 -0.0517 -89.6888>}]
}
constraints [
 {bone1Name=seat, bone2Name=Anim, swing1LimitDegrees=45, swing2LimitDegrees=45, twistLimitDegrees=45}
 {bone1Name=seat_back, bone2Name=seat, swing1LimitDegrees=45, swing2LimitDegrees=45, twistLimitDegrees=45}
]
